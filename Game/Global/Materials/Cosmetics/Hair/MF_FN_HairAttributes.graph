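# MF_FN_HairAttributes is a material function in the cosmetics hair folder that exposes per-vertex hair attributes as named outputs.
# It reads the second UV channel (index 1) and outputs it split as U and V, and whole as Root_UV.
# Seed comes from the vertex color green channel, while Tangent passes the Tangent input through unchanged.
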MaterialFunction /Game/Global/Materials/Cosmetics/Hair/MF_FN_HairAttributes
graph {nodes [
  n1 = TextureCoordinate() {coordinateIndex=1}
  FunctionOutput(None=ComponentMask(None=@n1) {r=true}) {outputName=U}
  FunctionOutput(None=ComponentMask(None=@n1) {g=true}) {outputName=V, sortPriority=1}
  FunctionOutput(None=VertexColor().G) {outputName=Seed, sortPriority=2}
  FunctionOutput(None=FunctionInput() {inputName=Tangent input}) {outputName=Tangent, sortPriority=3}
  FunctionOutput(None=TextureCoordinate() {coordinateIndex=1}) {outputName=Root_UV, sortPriority=4}
 ]}
