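# Material function for cosmetic hair that blends anisotropic hair shading over incoming material attributes, with a HairMask texture parameter.
# A "hair cards?" static switch selects between card and mesh treatments of color, normal, tangent, roughness and opacity.
# Cards use per-strand attributes to vary warped, scaled strand-texture UVs, noise-scraggled tangents and root transparency with reduced shadow opacity.
# Fresnel simulates light transmission, alongside optional subsurface shading and dithered pixel depth offset.
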MaterialFunction /Game/Global/Materials/Cosmetics/Hair/MF_FN_Hair
graph {nodes [
  ComponentMask() {r=true}
  n1 = TextureSampleParameter2D() {group=FN_Hair, parameterName=HairMask, texture=/Game/UI/Friends_UI/Social/white}
  n2 = LinearInterpolate(
   A=ScalarParameter() {defaultValue=0.3, group=FN_Hair, parameterName=SpeculaiMIN}
   B=ScalarParameter() {defaultValue=0.5, group=FN_Hair, parameterName=SpeculaiMAX}
   Alpha=StaticSwitchParameter(True=NamedRerouteUsage(), False=NamedRerouteUsage()) {group=FN_Hair, parameterName=hair cards?}
  )
  n3 = ScalarParameter() {defaultValue=0.5, group=FN_Hair, parameterName=Hair_anisotropy_min}
  n4 = ScalarParameter() {defaultValue=0.75, group=FN_Hair, parameterName=Hair_anisotropy_max}
  n5 = Saturate(None=NamedRerouteUsage())
  n6 = ComponentMask(None=TextureCoordinate() {coordinateIndex=1}) {g=true}
  n7 = NamedRerouteUsage()
  n8 = OneMinus(None=Fresnel(ExponentIn=Constant() {r=2}, BaseReflectFractionIn=Constant() {r=0.01}, Normal=VertexNormalWS()))
  n9 = MaterialFunctionCall!(Tangent input=NamedRerouteUsage()) {materialFunction=&/MF_FN_HairAttributes}
  n10 = OneMinus(None=ComponentMask(None=@n9.Root_UV) {g=true})
  n11 = ScalarParameter() {defaultValue=1.3, group=FN_Hair_haircards, parameterName=The transparency strength of the root}
  n12 = Min(A=ComponentMask(None=@n9.Root_UV) {g=true}, B=Saturate(None=Multiply(A=@n10, B=@n11)))
  n13 = Multiply(A=Multiply(A=Multiply(A=@n12) {constB=2}, B=NamedRerouteUsage()), B=Constant() {r=4}) {constB=6}
  n14 = StaticSwitchParameter(
   True=ShadowReplace(Default=Saturate(None=@n13), Shadow=ScalarParameter() {defaultValue=0.2, group=FN_Hair_haircards, parameterName=hair cards shaow intensity})
   False=Constant() {r=1}
  ) {group=FN_Hair, parameterName=hair cards?}
  n15 = NamedRerouteUsage()
  n16 = NamedRerouteUsage()
  n17 = Min(A=ComponentMask(None=@n9.Root_UV) {r=true}, B=OneMinus(None=ComponentMask(None=@n9.Root_UV) {r=true}))
  n18 = Multiply(A=Multiply(A=LinearInterpolate(Alpha=@n17) {constA=0.8}, B=@n16)) {constB=0.05}
  n19 = Multiply(A=@n16, B=Constant() {r=1.5}) {constB=0.5}
  n20 = Saturate(None=Multiply(
    A=LinearInterpolate(A=@n16, B=@n19, Alpha=@n8) {constB=0}
    B=ScalarParameter() {defaultValue=0.85, group=FN_Hair, parameterName=Hair mesh_SubstanceColor}
   ))
  n21 = LinearInterpolate(A=@n20, B=Multiply(A=@n20, B=NamedRerouteUsage()), Alpha=ScalarParameter() {group=FN_Hair, parameterName=Gmap_indensity})
  n22 = Add(A=@n9.U, B=@n9.U)
  n23 = ScalarParameter() {defaultValue=2, group=FN_Hair, parameterName=NoiseTiling}
  n24 = TextureSample(UVs=AppendVector(A=Multiply(A=@n22, B=@n23), B=Constant())) {texture=/Game/Global/Materials/Cosmetics/Hair/Textures/T_FNHair_Noise}
  n25 = MaterialFunctionCall(X=@n24.R, Y=Constant(), Z=@n24.R) {materialFunction=&MakeFloat3}
  n26 = Constant3Vector() {constant=<rgba 0.5 0.5 0 0>}
  n27 = Add(A=@n9.Tangent, B=Multiply(A=Subtract(A=@n25.Result, B=@n26), B=ScalarParameter() {defaultValue=0.3, group=FN_Hair, parameterName=Scraggle}))
  n28 = Constant()
  n29 = MaterialFunctionCall(X=@n28, Y=Constant(), Z=@n28) {materialFunction=&MakeFloat3}
  n30 = Constant3Vector() {constant=<rgba 0.5 0.5 0 1>}
  n31 = Add(A=@n9.Tangent, B=Multiply(A=Subtract(A=@n29.Result, B=@n30), B=ScalarParameter() {defaultValue=0.3, group=FN_Hair, parameterName=Scraggle}))
  n32 = Multiply(A=NamedRerouteUsage(), B=LinearInterpolate(A=Constant() {r=1.1}, B=Constant() {r=1}, Alpha=NamedRerouteUsage()) {constA=2})
  n33 = NamedRerouteUsage()
  n34 = GetMaterialAttributes(None=FunctionInput() {bUsePreviewValueAsDefault=true, inputType=FunctionInput_MaterialAttributes}) {attributeGetTypes [
    #1
    #2
    #3
    #4
    #5
    #6
    #7
    #8
    #9
   ]}
  n35 = SetMaterialAttributes!(
   MaterialAttributes=@n34.MaterialAttributes
   Roughness=LinearInterpolate(A=NamedRerouteUsage(), B=NamedRerouteUsage(), Alpha=NamedRerouteDeclaration(None=@n1.R) {name=HairMask, nodeColor=<rgba 0 0.0859 1 1>})
   Specular=LinearInterpolate(A=NamedRerouteUsage(), B=Saturate(None=@n2), Alpha=NamedRerouteUsage())
   Metallic=LinearInterpolate(A=NamedRerouteUsage(), B=ScalarParameter() {defaultValue=0.05, group=FN_Hair, parameterName=Metallic}, Alpha=NamedRerouteUsage())
   Anisotropy=LinearInterpolate(B=LinearInterpolate(A=@n3, B=@n4, Alpha=@n5), Alpha=NamedRerouteUsage())
   BaseColor=LinearInterpolate(A=NamedRerouteUsage(), B=StaticSwitchParameter(True=@n32, False=@n33) {group=FN_Hair, parameterName=hair cards?}, Alpha=NamedRerouteUsage())
   Normal=NamedRerouteUsage()
   ShadingModel=StaticSwitchParameter(True=ShadingModel() {shadingModel=MSM_Subsurface}, False=ShadingModel()) {group=FN_Hair_haircards, parameterName=Subsurface}
   PixelDepthOffset=NamedRerouteUsage()
   EmissiveColor=NamedRerouteUsage()
   Opacity=StaticSwitchParameter(True=LinearInterpolate(Alpha=@n6) {constA=0.9}, False=LinearInterpolate(B=@n8, Alpha=@n7)) {group=FN_Hair, parameterName=hair cards?}
   OpacityMask=Multiply(A=NamedRerouteUsage(), B=LinearInterpolate(B=@n14, Alpha=@n15) {constA=1})
   SubsurfaceColor=LinearInterpolate(B=StaticSwitchParameter(True=@n18, False=@n21) {group=FN_Hair, parameterName=hair cards?}, Alpha=NamedRerouteUsage())
   Tangent=LinearInterpolate(
    A=Constant3Vector() {constant=<rgba 0.2159 0.1878 0 1>}
    B=StaticSwitchParameter(True=@n27, False=@n31) {group=FN_Hair, parameterName=hair cards?}
    Alpha=NamedRerouteUsage()
   )
  ) {attributeSetTypes [
    #1
    #2
    #7
    #8
    #3
    #4
    #10
    #11
    #12
    #13
    #9
    #14
    #6
   ]}
  FunctionOutput(None=@n35)
  NamedRerouteDeclaration(None=ComponentMask(None=FunctionInput() {bUsePreviewValueAsDefault=true, inputName=Mask Input}) {g=true}) {name=Gmap, nodeColor=<rgba 0 1 0.1328 1>}
  NamedRerouteDeclaration(None=@n34.Tangent) {name=OutputTangent, nodeColor=<rgba 0.2656 0 1 1>}
  NamedRerouteDeclaration(None=@n34.Specular) {name=Out specular, nodeColor=<rgba 0 0.2266 1 1>}
  NamedRerouteDeclaration(None=@n34.Roughness) {name=OutPutRoughness, nodeColor=<rgba 1 0 0.6797 1>}
  NamedRerouteDeclaration(None=@n34.Anisotropy) {name=OutPutAnisotropy, nodeColor=<rgba 0.1016 0 1 1>}
  n36 = ScalarParameter() {defaultValue=1, group=FN_Hair, parameterName=Emissive_Brightness}
  n37 = Constant() {r=0.5}
  n38 = Constant() {r=3.5}
  n39 = VertexNormalWS()
  n40 = NamedRerouteUsage()
  n41 = StaticSwitchParameter(True=@n40, False=@n40) {group=FN_Hair, parameterName=hair cards?}
  n42 = Multiply(
   A=NamedRerouteDeclaration(None=MaterialFunctionCall(Brightness=@n37, Exponent=@n38, Vertex / Pixel Normal=@n39) {materialFunction=&MF_Fresnel_FX}.Fresnel Inverted Double Sided) {name=fresnel, nodeColor=<rgba 0 1 0.2969 1>}
   B=@n41
  )
  NamedRerouteDeclaration(None=LinearInterpolate(B=Multiply(A=@n42, B=@n36), Alpha=NamedRerouteUsage())) {name=simulate light transmission, nodeColor=<rgba 0 1 0.8594 1>}
  n43 = Multiply(A=MaterialFunctionCall() {materialFunction=&DitherTemporalAA}.Result, B=Constant())
  n44 = Multiply(
   A=MaterialFunctionCall() {materialFunction=&DitherTemporalAA}.Result
   B=ScalarParameter() {group=FN_Hair_haircards, parameterName="DepthOffsetDistance(MaskMode)"}
  )
  NamedRerouteDeclaration(None=LinearInterpolate(B=StaticSwitchParameter(True=@n44, False=@n43) {group=FN_Hair, parameterName=hair cards?}, Alpha=NamedRerouteUsage())) {name=Depth Offset, nodeColor=<rgba 1 0.2109 0 1>}
  n45 = MaterialFunctionCall(Normal=NamedRerouteUsage(), Flatness=ScalarParameter() {defaultValue=0.5, group=FN_Hair, parameterName=Hair Mesh Normal Flatness}) {materialFunction=&FlattenNormal}
  n46 = NamedRerouteUsage()
  n47 = MaterialFunctionCall(
   BaseNormal=@n45.Result
   AdditionalNormal=MaterialFunctionCall(
    Normal=TextureSample(UVs=@n46) {samplerType=SAMPLERTYPE_Normal, texture=/Game/Global/Materials/Cosmetics/Hair/Textures/T_FNHair__N}.RGB
    Flatness=ScalarParameter() {defaultValue=0.5, group=FN_Hair_haircards, parameterName=Hair strands Normal Flatness}
   ) {materialFunction=&FlattenNormal}.Result
  ) {materialFunction=&BlendAngleCorrectedNormals}
  NamedRerouteDeclaration(None=LinearInterpolate(
    A=NamedRerouteUsage()
    B=StaticSwitchParameter(True=@n47.Result, False=@n45.Result) {group=FN_Hair, parameterName=hair cards?}
    Alpha=NamedRerouteUsage()
   )) {name=Normal, nodeColor=<rgba 0.5 1 0 1>}
  NamedRerouteDeclaration(None=@n34.OpacityMask) {name=OutPutOpacityMask, nodeColor=<rgba 1 0 0.8203 1>}
  n48 = Saturate(None=Multiply(A=Add(A=@n42, B=@n41), B=ScalarParameter() {defaultValue=1, group=FN_Hair, parameterName=BaseColor_Brightness}) {constB=100})
  n49 = NamedRerouteUsage()
  NamedRerouteDeclaration(None=LinearInterpolate(A=@n48, B=Multiply(A=@n48, B=@n49), Alpha=ScalarParameter() {group=FN_Hair, parameterName=Gmap_indensity})) {name=basecolor, nodeColor=<rgba 0.8984 0 1 1>}
  n50 = TextureSample(UVs=Multiply(A=@n9.Root_UV, B=Constant() {r=2}) {constB=10}) {samplerType=SAMPLERTYPE_LinearGrayscale, texture=/Game/Global/Materials/Cosmetics/Hair/Textures/T_FNHair_Warp}
  n51 = Constant()
  n52 = Constant() {r=0.005}
  n53 = Constant() {r=0.5}
  n54 = ScalarParameter() {defaultValue=8, group=FN_Hair_haircards, parameterName=Hair thickness Max}
  n55 = LinearInterpolate(
   A=ScalarParameter() {defaultValue=4, group=FN_Hair_haircards, parameterName=Hair thickness Min}
   B=ScalarParameter() {defaultValue=9, group=FN_Hair_haircards, parameterName=Hair thickness Middle}
   Alpha=Saturate(None=@n9.Seed)
  )
  n56 = Add(
   A=Multiply(A=ComponentMask(None=@n50.RGB) {g=true, r=true}, B=LinearInterpolate(A=@n51, B=@n52, Alpha=@n9.Seed))
   B=MaterialFunctionCall(
    UVs=MaterialFunctionCall(UVs=@n9.Root_UV, "Rotation Angle (0-1)"=@n53) {materialFunction=&CustomRotator}.Rotated Values
    Texture Scale=If(A=@n9.Seed, "A > B"=@n54, "A == B"=@n55, "A < B"=@n55) {constB=0.8}
   ) {materialFunction=&ScaleUVsByCenter}.UVs
  )
  NamedRerouteDeclaration(None=TextureSampleParameter2D(UVs=NamedRerouteDeclaration(None=@n56) {name=HairStransUV, nodeColor=<rgba 0 0.5312 1 1>}) {group=FN_Hair_haircards, parameterName=hair_alpha, texture=/Game/Global/Materials/Cosmetics/Hair/Textures/T_FNHair_alpha_c}.R) {name=HairStradnsTexture, nodeColor=<rgba 0 1 0.625 1>}
  NamedRerouteDeclaration(None=@n34.BaseColor) {name=OutPutBaseColor, nodeColor=<rgba 0.2891 1 0 1>}
  NamedRerouteDeclaration(None=@n34.Normal) {name=OutputNM, nodeColor=<rgba 1 0.5625 0 1>}
  NamedRerouteDeclaration(None=@n34.Metallic) {name=OutPutMetallic, nodeColor=<rgba 0 1 1 1>}
  n57 = StaticSwitchParameter(
   True=ScalarParameter() {defaultValue=0.65, group=FN_Hair, parameterName=HairRoughness}
   False=ScalarParameter() {defaultValue=0.75, group=FN_Hair, parameterName=HairRoughness_Strands}
  ) {group=FN_Hair, parameterName=hair cards?}
  n58 = Add(A=ScalarParameter() {defaultValue=0.4, group=FN_Hair, parameterName=RoughnessVariation})
  n59 = Divide(B=@n58)
  n60 = Multiply(A=TextureCoordinate(), B=ScalarParameter() {defaultValue=10, group=FN_Hair, parameterName=RoughnessNoise})
  n61 = NamedRerouteUsage()
  n62 = Multiply(
   A=LinearInterpolate(
    A=ScalarParameter() {defaultValue=0.5, group=FN_Hair, parameterName=Hair_roughness_min}
    Alpha=TextureSample(UVs=@n60) {texture=/Game/Global/Textures/Default/Noise/Noise}.R
   ) {constA=0.5}
   B=StaticSwitchParameter(True=Add(A=@n9.Seed, B=@n61), False=NamedRerouteUsage()) {group=FN_Hair, parameterName=hair cards?}
  )
  NamedRerouteDeclaration(None=MaterialFunctionCall!(1=Power(Base=@n57, Exp=@n58), 2=@n57, 3=Power(Base=@n57, Exp=@n59), 4, A=Saturate(None=@n62)) {materialFunction=/Engine/Functions/Engine_MaterialFunctions02/Lerp_Multiple_Float}.Lerp 3 Inputs) {name=Roughness, nodeColor=<rgba 1 0.4453 0 1>}
  TextureSampleParameter2D() {group=FN_Hair, parameterName="Skin_FX_opacity_mask(MaskMode)", texture=/Game/ContentCreationTools/ObjectPlacement/Textures/T_BlankWhite_01}
 ]}
