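# M_HairParent_2023 is a parent material for character hair, usable on skeletal, clothing, morph target and Nanite meshes, with shading model chosen through material attributes.
# A shared character PreFX function feeds a fur normal and UV blend masked by SkinFX_Mask, then the MF_FN_Hair function.
# Static switch UseFNHair, on by default, selects that chain or raw PreFX attributes.
# The result passes into the shared character PostFX function.
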Material /Game/Global/Materials/Cosmetics/Hair/M_HairParent_2023
graph {
 properties {
  bUseMaterialAttributes=true
  bUsedWithClothing=true
  bUsedWithMeshDeformer=true
  bUsedWithMorphTargets=true
  bUsedWithNanite=true
  bUsedWithSkeletalMesh=true
  pixelDepthOffsetMode=PDOM_Legacy
  shadingModel=MSM_FromMaterialExpression
  usedShadingModels=Default Lit | Subsurface | Clear Coat | Subsurface Profile | Cloth
 }
 nodes [
  n1 = MaterialFunctionCall() {materialFunction=&MF_CharacterParent_2023_PreFX}
  n2 = MaterialFunctionCall!(FurOpacityMaskTexture=TextureObjectParameter() {group=Skin FX, parameterName=SkinFX_Mask}, In=@n1.Result) {materialFunction=/Game/Weapons/WeaponSkins/MaterialFunction/MF_FurNormalAndUVBlend}
  MaterialFunctionCall(
   Material=StaticSwitchParameter(
    True=MaterialFunctionCall!(Mask Input=@n1.M, In=@n2.Result) {materialFunction=&/MF_FN_Hair}.Result
    False=@n1.Result
   ) {defaultValue=true, group=Skin FX, parameterName=UseFNHair}
   M=@n1.M
   Pre Skinned Local Position=@n1.M
   Distance Blend=@n1.M
   Fresnel=@n1.M
  ) {materialFunction=&MF_CharacterParent_2023_PostFX}
 ]
}
